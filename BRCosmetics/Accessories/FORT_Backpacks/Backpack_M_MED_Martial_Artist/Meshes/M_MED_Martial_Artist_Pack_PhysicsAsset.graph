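# Physics asset for the Martial Artist back bling in BRCosmetics, setting up secondary motion for its bottle and fan.
# A box body represents the pack, with a capsule for the bottle and a long box for the fan.
# Both dangling bodies are constrained directly to the pack with tight swing and twist limits, the fan allowed slightly wider swing, so they can only wobble subtly.
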PhysicsAsset /BRCosmetics/Accessories/FORT_Backpacks/Backpack_M_MED_Martial_Artist/Meshes/M_MED_Martial_Artist_Pack_PhysicsAsset
bodies {
 dyn_bottle [{shapeType=Capsule, length=8.5995, radius=3.797, center=<0.5606 0.1584 2.2614>, rotation=<rot 0 -90.2029 -85.4133>}]
 dyn_fan [{shapeType=Box, extentX=3.978, extentY=3.4085, extentZ=24.1743, center=<3.6984 0.2358 1.0879>, rotation=<rot 0 91.1189 -89.9761>}]
 pack [{shapeType=Box, extentX=16.7267, extentY=7.5, extentZ=7.5, center=<0.3798 0 0>}]
}
constraints [
 {bone1Name=dyn_bottle, bone2Name=pack, swing1LimitDegrees=5, swing2LimitDegrees=6.5, twistLimitDegrees=2.5}
 {bone1Name=dyn_fan, bone2Name=pack, swing1LimitDegrees=10, swing2LimitDegrees=10, twistLimitDegrees=2.5}
]
